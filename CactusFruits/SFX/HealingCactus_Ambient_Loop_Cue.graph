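# Ambient Sound Cue for the Healing Cactus, mixing a continuous looping bed with intermittent random one-shots in the Fort_Ambient_Athena sound class.
# The bed is pitched down and faded in by an envelope, with its own loop attenuation.
# The one-shot branch loops through a random delay and pitch/volume variation, using separate attenuation.
# A quality-level node selects a four- or three-variant random pool, and a concurrency override caps instances.
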SoundCue /CactusFruits/SFX/HealingCactus_Ambient_Loop_Cue
root=Mixer_0
nodes [
 {
  id=Attenuation_0
  children [Enveloper_0]
  properties {attenuationSettings=/ChronoEnviroItems/LaunchablePlants/Gameplay/Assets/Sounds/Launchable_Plants_Ambient_Loop_Att}
 }
 {
  id=Attenuation_2
  children [Looping_0]
  properties {attenuationSettings=/ChronoEnviroItems/LaunchablePlants/Gameplay/Assets/Sounds/Launchable_Plants_Ambient_OneShot_Att}
 }
 {id=Delay_0, children [Modulator_1], properties {delayMax=6, delayMin=2}}
 {
  id=Enveloper_0
  children [Modulator_0]
  properties {pitchCurve {editorCurveData {keys [{time=0}]}}, volumeCurve {editorCurveData {keys [<at 0 0 RCIM_Cubic>,<at 1 1>]}}}
 }
 {id=Looping_0, children [Delay_0]}
 {id=Mixer_0, children [Attenuation_0, Attenuation_2], properties {inputVolume [0.7,1]}}
 {id=Modulator_0, children [WavePlayer_0], properties {pitchMax=0.7, pitchMin=0.7}}
 {id=Modulator_1, children [QualityLevel_0], properties {pitchMax=1.2, pitchMin=0.98, volumeMax=1.02, volumeMin=0.98}}
 {id=QualityLevel_0, children [Random_2,Random_2,Random_1,Random_1]}
 {id=Random_1, children [WavePlayer_2,WavePlayer_3,WavePlayer_4], properties {bShouldExcludeFromBranchCulling=true, weights [1,1,1]}}
 {id=Random_2, children [WavePlayer_1,WavePlayer_2,WavePlayer_3,WavePlayer_4], properties {weights [1,1,1,1]}}
 {id=WavePlayer_0, properties {bLooping=true, soundWaveAssetPtr=&/HealingCactus_Ambient_Loop_01}}
 {id=WavePlayer_1, properties {soundWaveAssetPtr=&/HealingCactus_Amb_OneShots_01}}
 {id=WavePlayer_2, properties {soundWaveAssetPtr=&/HealingCactus_Amb_OneShots_02}}
 {id=WavePlayer_3, properties {soundWaveAssetPtr=&/HealingCactus_Amb_OneShots_03}}
 {id=WavePlayer_4, properties {soundWaveAssetPtr=&/HealingCactus_Amb_OneShots_04}}
]
properties {
 attenuationSettings=&/HealingCactus_Ambient_Loop_Att
 bOverrideConcurrency=true
 maxDistance=2000
 soundClassObject=&Fort_Ambient_Athena
 volumeMultiplier=1.253
 concurrencyOverrides {maxCount=3}
 concurrencySet [null]
}
