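# Sound cue for Healing Cactus impacts, using player pickaxe-impact attenuation, a melee weapon sound class and the rifle bullet-impact concurrency set.
# A mixer combines two layers, each picking at random from eight equally weighted waves with slight pitch and volume modulation.
# The "Lyr" layer starts after a short fixed delay, trailing the main impact.
# The mixer's third input is empty.
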SoundCue /CactusFruits/SFX/HealingCactus_Impact_Cue
root=Mixer_0
nodes [
 {id=Delay_0, children [Modulator_0], properties {delayMax=0.1, delayMin=0.1}}
 {id=Mixer_0, children [Modulator_3,Delay_0,null], properties {inputVolume [1,1,1.25]}}
 {id=Modulator_0, children [Random_0], properties {pitchMax=1.02, pitchMin=0.98, volumeMax=1.02, volumeMin=1}}
 {id=Modulator_3, children [Random_4], properties {pitchMax=1.02, pitchMin=0.98, volumeMax=1.02, volumeMin=0.98}}
 {
  id=Random_0
  children [WavePlayer_0, WavePlayer_1, WavePlayer_10, WavePlayer_11, WavePlayer_12, WavePlayer_13, WavePlayer_14, WavePlayer_15]
  properties {weights [1,1,1,1,1,1,1,1]}
 }
 {
  id=Random_4
  children [WavePlayer_2,WavePlayer_3,WavePlayer_4,WavePlayer_5,WavePlayer_6,WavePlayer_7,WavePlayer_8,WavePlayer_9]
  properties {weights [1,1,1,1,1,1,1,1]}
 }
 {id=WavePlayer_0, properties {soundWaveAssetPtr=&/HealingCactus_Impact_01_Lyr}}
 {id=WavePlayer_1, properties {soundWaveAssetPtr=&/HealingCactus_Impact_02_Lyr}}
 {id=WavePlayer_10, properties {soundWaveAssetPtr=&/HealingCactus_Impact_03_Lyr}}
 {id=WavePlayer_11, properties {soundWaveAssetPtr=&/HealingCactus_Impact_04_Lyr}}
 {id=WavePlayer_12, properties {soundWaveAssetPtr=&/HealingCactus_Impact_05_Lyr}}
 {id=WavePlayer_13, properties {soundWaveAssetPtr=&/HealingCactus_Impact_06_Lyr}}
 {id=WavePlayer_14, properties {soundWaveAssetPtr=&/HealingCactus_Impact_07_Lyr}}
 {id=WavePlayer_15, properties {soundWaveAssetPtr=&/HealingCactus_Impact_08_Lyr}}
 {id=WavePlayer_2, properties {soundWaveAssetPtr=&/HealingCactus_Impact_01}}
 {id=WavePlayer_3, properties {soundWaveAssetPtr=&/HealingCactus_Impact_02}}
 {id=WavePlayer_4, properties {soundWaveAssetPtr=&/HealingCactus_Impact_03}}
 {id=WavePlayer_5, properties {soundWaveAssetPtr=&/HealingCactus_Impact_04}}
 {id=WavePlayer_6, properties {soundWaveAssetPtr=&/HealingCactus_Impact_05}}
 {id=WavePlayer_7, properties {soundWaveAssetPtr=&/HealingCactus_Impact_06}}
 {id=WavePlayer_8, properties {soundWaveAssetPtr=&/HealingCactus_Impact_07}}
 {id=WavePlayer_9, properties {soundWaveAssetPtr=&/HealingCactus_Impact_08}}
]
properties {
 attenuationSettings=/Game/Sounds/Attenuation/Fort_Weapon_Pickaxe_Player_Impact_Att
 maxDistance=3512
 soundClassObject=&Fort_WeaponMelee_PSM
 concurrencyOverrides {maxCount=4, resolutionRule=PreventNew}
 concurrencySet [&RifleBulletImpact_Conc]
}
