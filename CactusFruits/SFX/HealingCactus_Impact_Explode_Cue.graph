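# Sound cue for the Healing Cactus fruit's impact explosion, spatialized with its explode attenuation and sharing rifle bullet impact concurrency.
# A mixer layers a randomized explosion, an impact layer with slight pitch and volume variation, and two delayed heal sounds.
# The heal layers draw on like-named variants, one pitched up and louder in the mix, the other pitched down.
# A quality-level switch selects either eight explosion variants or a three-variant subset.
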SoundCue /CactusFruits/SFX/HealingCactus_Impact_Explode_Cue
root=Mixer_1
nodes [
 {id=Delay_0, children [Modulator_2], properties {delayMax=0.5, delayMin=0.5}}
 {id=Delay_1, children [Modulator_1], properties {delayMax=0.5, delayMin=0.3}}
 {id=Mixer_1, children [Modulator_0,Delay_0,Delay_1,Modulator_3], properties {inputVolume [0.5,2,1,1]}}
 {id=Modulator_0, children [QualityLevel_0]}
 {id=Modulator_1, children [Random_1], properties {pitchMax=0.85, pitchMin=0.75}}
 {id=Modulator_2, children [Random_3], properties {pitchMax=1.35, pitchMin=1.2}}
 {id=Modulator_3, children [Random_4], properties {pitchMax=1.02, pitchMin=0.98, volumeMax=1.02, volumeMin=1}}
 {id=QualityLevel_0, children [Random_0,Random_0,Random_2,Random_2]}
 {
  id=Random_0
  children [WavePlayer_14, WavePlayer_15, WavePlayer_16, WavePlayer_17, WavePlayer_20, WavePlayer_21, WavePlayer_22, WavePlayer_23]
  properties {weights [1,1,1,1,1,1,1,1]}
 }
 {id=Random_1, children [WavePlayer_24, WavePlayer_25, WavePlayer_26, WavePlayer_27], properties {weights [1,1,1,1]}}
 {id=Random_2, children [WavePlayer_21, WavePlayer_22, WavePlayer_23], properties {bShouldExcludeFromBranchCulling=true, weights [1,1,1]}}
 {id=Random_3, children [WavePlayer_0, WavePlayer_1, WavePlayer_18, WavePlayer_19], properties {weights [1,1,1,1]}}
 {
  id=Random_4
  children [WavePlayer_2,WavePlayer_3,WavePlayer_4,WavePlayer_5,WavePlayer_6,WavePlayer_7,WavePlayer_8,WavePlayer_9]
  properties {weights [1,1,1,1,1,1,1,1]}
 }
 {id=WavePlayer_0, properties {soundWaveAssetPtr=&/HealingCactus_Explode_Heal_04}}
 {id=WavePlayer_1, properties {soundWaveAssetPtr=&/HealingCactus_Explode_Heal_01}}
 {id=WavePlayer_14, properties {soundWaveAssetPtr=&/HealingCactus_Impact_Explode_01}}
 {id=WavePlayer_15, properties {soundWaveAssetPtr=&/HealingCactus_Impact_Explode_02}}
 {id=WavePlayer_16, properties {soundWaveAssetPtr=&/HealingCactus_Impact_Explode_03}}
 {id=WavePlayer_17, properties {soundWaveAssetPtr=&/HealingCactus_Impact_Explode_04}}
 {id=WavePlayer_18, properties {soundWaveAssetPtr=&/HealingCactus_Explode_Heal_02}}
 {id=WavePlayer_19, properties {soundWaveAssetPtr=&/HealingCactus_Explode_Heal_03}}
 {id=WavePlayer_2, properties {soundWaveAssetPtr=&/HealingCactus_Impact_01_Lyr}}
 {id=WavePlayer_20, properties {soundWaveAssetPtr=&/HealingCactus_Impact_Explode_05}}
 {id=WavePlayer_21, properties {soundWaveAssetPtr=&/HealingCactus_Impact_Explode_06}}
 {id=WavePlayer_22, properties {soundWaveAssetPtr=&/HealingCactus_Impact_Explode_07}}
 {id=WavePlayer_23, properties {soundWaveAssetPtr=&/HealingCactus_Impact_Explode_08}}
 {id=WavePlayer_24, properties {soundWaveAssetPtr=&HealingCactus_Explode_Heal_01}}
 {id=WavePlayer_25, properties {soundWaveAssetPtr=&HealingCactus_Explode_Heal_02}}
 {id=WavePlayer_26, properties {soundWaveAssetPtr=&HealingCactus_Explode_Heal_03}}
 {id=WavePlayer_27, properties {soundWaveAssetPtr=&HealingCactus_Explode_Heal_04}}
 {id=WavePlayer_3, properties {soundWaveAssetPtr=&/HealingCactus_Impact_02_Lyr}}
 {id=WavePlayer_4, properties {soundWaveAssetPtr=&/HealingCactus_Impact_03_Lyr}}
 {id=WavePlayer_5, properties {soundWaveAssetPtr=&/HealingCactus_Impact_04_Lyr}}
 {id=WavePlayer_6, properties {soundWaveAssetPtr=&/HealingCactus_Impact_05_Lyr}}
 {id=WavePlayer_7, properties {soundWaveAssetPtr=&/HealingCactus_Impact_06_Lyr}}
 {id=WavePlayer_8, properties {soundWaveAssetPtr=&/HealingCactus_Impact_07_Lyr}}
 {id=WavePlayer_9, properties {soundWaveAssetPtr=&/HealingCactus_Impact_08_Lyr}}
]
properties {
 attenuationSettings=&/HealingCactus_Explode_Att
 maxDistance=3512
 soundClassObject=/Game/Sounds/FortSoundClassesAndModes/Fort_Explosives_Medium_PSM
 concurrencyOverrides {maxCount=4, resolutionRule=PreventNew}
 concurrencySet [&RifleBulletImpact_Conc]
}
