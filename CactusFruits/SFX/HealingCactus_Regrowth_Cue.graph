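# Sound cue for the healing cactus regrowth effect, layering a regrowth sound with a delayed fruit sound.
# A mixer plays one of four random regrowth layer waves, slightly louder, alongside one of three random fruit waves that starts about 1.5 seconds later.
# Both branches get subtle pitch and volume modulation for variety.
# It uses ambient attenuation and sound class, with a concurrency override capping simultaneous instances.
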SoundCue /CactusFruits/SFX/HealingCactus_Regrowth_Cue
root=Mixer_0
nodes [
 {id=Delay_0, children [Modulator_0], properties {delayMax=1.5, delayMin=1.4}}
 {id=Mixer_0, children [Modulator_1,Delay_0], properties {inputVolume [1.5,1]}}
 {id=Modulator_0, children [Random_1], properties {pitchMax=1.02, pitchMin=0.98, volumeMax=1.02, volumeMin=0.98}}
 {id=Modulator_1, children [Random_0], properties {pitchMax=1.02, pitchMin=0.98, volumeMax=1.02, volumeMin=0.98}}
 {
  id=Random_0
  children [WavePlayer_0,WavePlayer_1,WavePlayer_2,WavePlayer_3]
  properties {bShouldExcludeFromBranchCulling=true, weights [1,1,1,1]}
 }
 {id=Random_1, children [WavePlayer_4,WavePlayer_5,WavePlayer_6], properties {weights [1,1,1]}}
 {id=WavePlayer_0, properties {soundWaveAssetPtr=&/HealingCactus_Regrowth_Lyr_01}}
 {id=WavePlayer_1, properties {soundWaveAssetPtr=&/HealingCactus_Regrowth_Lyr_02}}
 {id=WavePlayer_2, properties {soundWaveAssetPtr=&/HealingCactus_Regrowth_Lyr_03}}
 {id=WavePlayer_3, properties {soundWaveAssetPtr=&/HealingCactus_Regrowth_Lyr_04}}
 {id=WavePlayer_4, properties {soundWaveAssetPtr=&/HealingCactus_Regrowth_Fruit_01}}
 {id=WavePlayer_5, properties {soundWaveAssetPtr=&/HealingCactus_Regrowth_Fruit_02}}
 {id=WavePlayer_6, properties {soundWaveAssetPtr=&/HealingCactus_Regrowth_Fruit_03}}
]
properties {
 attenuationSettings=&/HealingCactus_Ambient_Loop_Att
 bOverrideConcurrency=true
 maxDistance=1750
 soundClassObject=&Fort_Ambient_Athena
 volumeMultiplier=1.504
 concurrencyOverrides {maxCount=3}
 concurrencySet [null]
}
